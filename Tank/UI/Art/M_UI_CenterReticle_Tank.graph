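# Translucent UI material that draws the line shapes of the tank's center reticle procedurally.
# A deprecated symmetry function mirrors the UVs, which are offset into a center line and a top/bottom pair sharing one negated Y offset.
# Each line is a Custom-node box signed distance field, with its width scaled by an Intro parameter.
# SmoothStep sharpens the fields, and scalar dot products multiply them before inversion merges the lines into one mask.
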Material /Tank/UI/Art/M_UI_CenterReticle_Tank
graph {
 properties {blendMode=BLEND_Translucent, materialDomain=MD_UI, pixelDepthOffsetMode=PDOM_Legacy, usedShadingModels=Default Lit}
 nodes [
  Constant() {r=1}
  n1 = ScalarParameter() {defaultValue=0.01, parameterName=Sharpness}
  n2 = MaterialFunctionCall!(UV, Offset=AppendVector(A=ScalarParameter() {parameterName=XOffset}, B=Constant()), Rotation, Scale, MirrorY=StaticBool() {value=true}) {materialFunction=/FortUILibrary/MaterialFunctions/Basics/MF_UI_Symmetry_2x1_DEPRECATED}
  n3 = ScalarParameter() {defaultValue=0.5, parameterName=CenterLine_X}
  n4 = ScalarParameter() {parameterName=CenterLine_Y}
  MaterialExpressionConstantBiasScale_0 = ConstantBiasScale(None=Add(A=@n2.Result, B=AppendVector(A=@n3, B=@n4))) {bias=-0.5, scale=2}
  MaterialExpressionAppendVector_16 = AppendVector(
   A=LinearInterpolate(
    B=ScalarParameter() {defaultValue=0.66, parameterName=Line SIxe X}
    Alpha=ScalarParameter() {defaultValue=1, parameterName=Intro, sliderMax=1}
   )
   B=ScalarParameter() {defaultValue=0.0005, parameterName=Line Size Y}
  )
  n5 = Custom!(p=@MaterialExpressionConstantBiasScale_0, b=@MaterialExpressionAppendVector_16) {
   code="float2 d = abs(p)-b;\r\n    return length(max(d,0.0)) + min(max(d.x,d.y),0.0);"
   outputType=CMOT_Float1
   inputs [
    {inputName=p, input {expression=&self:MaterialExpressionConstantBiasScale_0}}
    {inputName=b, input {expression=&self:MaterialExpressionAppendVector_16}}
   ]
  }
  n6 = ScalarParameter() {parameterName=Top_Bot_X}
  n7 = ScalarParameter() {defaultValue=0.25, parameterName=Top_Bot_Y}
  MaterialExpressionConstantBiasScale_1 = ConstantBiasScale(None=Add(A=@n2.Result, B=AppendVector(A=@n6, B=@n7))) {bias=-0.5, scale=2}
  n8 = Custom(p=@MaterialExpressionConstantBiasScale_1, b=@MaterialExpressionAppendVector_16) {
   code="float2 d = abs(p)-b;\r\n    return length(max(d,0.0)) + min(max(d.x,d.y),0.0);"
   outputType=CMOT_Float1
   inputs [
    {inputName=p, input {expression=&self:MaterialExpressionConstantBiasScale_1}}
    {inputName=b, input {expression=&self:MaterialExpressionAppendVector_16}}
   ]
  }
  n9 = SmoothStep(Max=ScalarParameter() {defaultValue=0.01, parameterName=Sharpness}, Value=@n8)
  n10 = Multiply(A=@n7) {constB=-1}
  MaterialExpressionConstantBiasScale_2 = ConstantBiasScale(None=Add(A=@n2.Result, B=AppendVector(A=@n6, B=@n10))) {bias=-0.5, scale=2}
  n11 = Custom(p=@MaterialExpressionConstantBiasScale_2, b=@MaterialExpressionAppendVector_16) {
   code="float2 d = abs(p)-b;\r\n    return length(max(d,0.0)) + min(max(d.x,d.y),0.0);"
   outputType=CMOT_Float1
   inputs [
    {inputName=p, input {expression=&self:MaterialExpressionConstantBiasScale_2}}
    {inputName=b, input {expression=&self:MaterialExpressionAppendVector_16}}
   ]
  }
  n12 = SmoothStep(Max=ScalarParameter() {defaultValue=0.01, parameterName=Sharpness}, Value=@n11)
  OneMinus(None=DotProduct(A=SmoothStep(Max=@n1, Value=@n5), B=DotProduct(A=@n9, B=@n12)))
 ]
}
