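# Translucent UI master material for the backing of a segmented tank progress bar, driven along the V axis.
# Progress, scaled by Progress_Anim, is truncated to whole segments and turned into a sharp fill mask that gates the background color's alpha.
# An absolute sine of offset UVs draws procedural separators with adjustable count, position, thickness and sharpness.
# The separator mask multiplies that alpha, with LiftSegmentAlpha setting the level inside the gaps.
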Material /Tank/UI/Art/M_UI_TankProgressBar_Backing_Master
graph {
 properties {blendMode=BLEND_Translucent, materialDomain=MD_UI, pixelDepthOffsetMode=PDOM_Legacy, usedShadingModels=Default Lit}
 nodes [
  n1 = Multiply(
   A=Add(A=TextureCoordinate(), B=ScalarParameter() {group=Separator, parameterName=SeparatorPosition}) {constB=0}
   B=Multiply(A=ScalarParameter() {defaultValue=10, group=Separator, parameterName=Separator Number}) {constB=0.5}
  ) {constB=10}
  n2 = ComponentMask(None=Abs(None=Sine(None=@n1))) {g=true}
  n3 = ScalarParameter() {defaultValue=0.1, group=Separator, parameterName=Separators_Thickness}
  n4 = Multiply(A=Saturate(None=Subtract(A=@n2, B=@n3)), B=ScalarParameter() {defaultValue=128, group=Separator, parameterName=Separator_Sharpness}) {constB=256}
  Multiply(
   A=LinearInterpolate(B=VectorParameter() {parameterName=RGBA_Background, defaultValue=<rgba 1 1 1 0.33>}.A, Alpha=NamedRerouteUsage())
   B=LinearInterpolate(A=ScalarParameter() {group=Separator, parameterName=LiftSegmentAlpha}, Alpha=Saturate(None=@n4))
  )
  n5 = LinearInterpolate(
   B=ScalarParameter() {defaultValue=0.5, parameterName=Progress, sliderMax=1}
   Alpha=ScalarParameter() {defaultValue=1, parameterName=Progress_Anim}
  )
  n6 = ScalarParameter() {defaultValue=10, group=Separator, parameterName=Separator Number}
  n7 = Truncate!(None=Multiply(A=OneMinus(None=@n5), B=@n6)) {is=class MaterialExpressionTruncate}
  n8 = ComponentMask(None=Subtract(A=TextureCoordinate(), B=Divide(A=@n7, B=@n6))) {g=true}
  n9 = ScalarParameter() {defaultValue=1024, group=Separator, parameterName=Progress_Sharpness}
  NamedRerouteDeclaration(None=Saturate(None=Multiply(A=@n8, B=@n9))) {name=Progress, nodeColor=<rgba 0 1 0.3906 1>}
 ]
}
