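# Additive UI master material for the tank's segmented progress bar, with progress, warning, boost and overheat parameters.
# Truncate snaps progress and threshold positions to segment steps, then sharpened UV V-axis gradients become named reroute masks.
# An absolute sine of the UVs cuts separators into opacity and shapes the inner glow mask.
# Nested lerps driven by reroutes and BoostOn blend background, progress and warning colors, alongside a time-based Blink reroute.
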Material /Tank/UI/Art/M_UI_TankProgressBar_Master
graph {
 properties {blendMode=BLEND_Additive, materialDomain=MD_UI, pixelDepthOffsetMode=PDOM_Legacy, usedShadingModels=Default Lit}
 nodes [
  n1 = VectorParameter() {parameterName=RGBA_Progress_Out, defaultValue=<rgba 1 0.5029 0 1>}
  n2 = VectorParameter() {parameterName=RGBA_Progress_In, defaultValue=<rgba 1 0.6914 0 1>}
  n3 = LinearInterpolate(
   A=@n2.RGB
   B=LinearInterpolate(A=VectorParameter() {parameterName=RGBA_Progress_Boost, defaultValue=<rgba 1 0.7761 0.2743 1>}.RGB, B=@n2.RGB, Alpha=NamedRerouteUsage())
   Alpha=ScalarParameter() {parameterName=BoostOn, sliderMax=1}
  )
  n4 = NamedRerouteUsage()
  n5 = VectorParameter() {parameterName=RGBA_Warning_Out, defaultValue=<rgba 1 0 0 1>}
  n6 = VectorParameter() {parameterName=RGBA_Warning_In, defaultValue=<rgba 1 0.112 0 1>}
  n7 = LinearInterpolate(
   A=@n6.RGB
   B=LinearInterpolate(A=@n2.RGB, B=@n6.RGB, Alpha=NamedRerouteUsage())
   Alpha=ScalarParameter() {parameterName=BoostOn, sliderMax=1}
  )
  n8 = NamedRerouteUsage()
  n9 = Add(A=NamedRerouteUsage(), B=NamedRerouteUsage())
  LinearInterpolate(
   A=VectorParameter() {parameterName=RGBA_Background, defaultValue=<rgba 1 1 1 0.05>}.RGB
   B=LinearInterpolate(A=LinearInterpolate(A=@n1.RGB, B=@n3, Alpha=@n4), B=LinearInterpolate(A=@n5.RGB, B=@n7, Alpha=@n8), Alpha=Saturate(None=@n9))
   Alpha=NamedRerouteUsage()
  )
  n10 = Add(A=TextureCoordinate(), B=ScalarParameter() {group=Separator, parameterName=SeparatorPosition}) {constB=0}
  n11 = Multiply(A=ScalarParameter() {defaultValue=10, group=Separator, parameterName=Separator Number}) {constB=0.5}
  n12 = Abs(None=Sine(None=Multiply(A=@n10, B=@n11) {constB=10}))
  n13 = Saturate(None=Subtract(A=ComponentMask(None=@n12) {g=true}, B=ScalarParameter() {defaultValue=0.1, group=Separator, parameterName=Separators_Thickness}))
  n14 = Multiply(A=@n13, B=ScalarParameter() {defaultValue=128, group=Separator, parameterName=Separator_Sharpness}) {constB=256}
  Multiply(
   A=LinearInterpolate(
    A=VectorParameter() {parameterName=RGBA_Background, defaultValue=<rgba 1 1 1 0.05>}.A
    B=ScalarParameter() {defaultValue=1, parameterName=Progress_Opacity, sliderMax=1}
    Alpha=NamedRerouteUsage()
   )
   B=LinearInterpolate(A=ScalarParameter() {group=Separator, parameterName=LiftSegmentAlpha}, Alpha=Saturate(None=@n14))
  )
  n15 = LinearInterpolate(
   B=ScalarParameter() {defaultValue=1, parameterName=Progress, sliderMax=1}
   Alpha=ScalarParameter() {defaultValue=1, parameterName=Progress_Anim}
  )
  n16 = ScalarParameter() {defaultValue=10, group=Separator, parameterName=Separator Number}
  n17 = Truncate!(None=Multiply(A=OneMinus(None=@n15), B=@n16)) {is=class MaterialExpressionTruncate}
  n18 = ComponentMask(None=Subtract(A=TextureCoordinate(), B=Divide(A=@n17, B=@n16))) {g=true}
  n19 = ScalarParameter() {defaultValue=1024, group=Separator, parameterName=Progress_Sharpness}
  NamedRerouteDeclaration(None=Saturate(None=Multiply(A=@n18, B=@n19))) {name=Progress, nodeColor=<rgba 0 1 0.3906 1>}
  n20 = ScalarParameter() {defaultValue=0.2, parameterName=Warning_Position, sliderMax=1}
  n21 = ScalarParameter() {defaultValue=10, group=Separator, parameterName=Separator Number}
  n22 = Truncate(None=Multiply(A=OneMinus(None=@n20), B=@n21))
  n23 = ComponentMask(None=Subtract(A=TextureCoordinate(), B=Divide(A=@n22, B=@n21))) {g=true}
  n24 = ScalarParameter() {defaultValue=1024, group=Separator, parameterName=Progress_Sharpness}
  NamedRerouteDeclaration(None=Saturate(None=Multiply(A=@n23, B=@n24))) {name=Warning, nodeColor=<rgba 0.7813 0 1 1>}
  n25 = Saturate(None=Multiply(A=@n13, B=ScalarParameter() {defaultValue=0.1, group=Inner Glow, parameterName=InnerGlow_Sharpness}) {constB=256})
  n26 = Sine(None=TextureCoordinate()) {period=2}
  n27 = ScalarParameter() {defaultValue=-0.25, group=Inner Glow, parameterName=InnerGlow_Position_Sides}
  n28 = Saturate(None=ComponentMask(None=Add(A=@n26, B=@n27)) {r=true})
  NamedRerouteDeclaration(None=Multiply(A=Multiply(A=@n25, B=@n28), B=ScalarParameter() {defaultValue=10, group=Inner Glow, parameterName=Inner Glow strenght})) {name=InnerGlow_Mask, nodeColor=<rgba 1 0.8672 0 1>}
  n29 = TextureCoordinate()
  n30 = ScalarParameter() {defaultValue=0.2, parameterName=Overheat_Position, sliderMax=1}
  n31 = ScalarParameter() {defaultValue=10, group=Separator, parameterName=Separator Number}
  n32 = Divide(A=Truncate(None=Multiply(A=@n30, B=@n31)), B=@n31)
  n33 = Multiply(
   A=ComponentMask(None=Subtract(A=@n29, B=@n32)) {g=true}
   B=ScalarParameter() {defaultValue=1024, group=Separator, parameterName=Progress_Sharpness}
  )
  NamedRerouteDeclaration(None=Saturate(None=OneMinus(None=@n33))) {name=OverHeat, nodeColor=<rgba 0 0.6953 1 1>}
  n34 = VectorParameter() {group=Pulse, parameterName=PulseConstraints, defaultValue=<rgba 0 1 0 1>}
  n35 = Time()
  n36 = ScalarParameter() {defaultValue=1, group=Pulse, parameterName=PulseSpeed}
  n37 = Add(A=Fmod(A=Multiply(A=@n35, B=@n36), B=Constant() {r=1}), B=ScalarParameter() {parameterName=PulseTimeOffset})
  NamedRerouteDeclaration(None=LinearInterpolate(
    A=@n34.R
    B=@n34.R
    Alpha=MaterialFunctionCall(Value=@n37) {materialFunction=&LinearSine}.Linear Sine
   )) {name=Blink, nodeColor=<rgba 0.1719 0 1 1>}
 ]
}
