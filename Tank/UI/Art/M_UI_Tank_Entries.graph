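# Additive UI material for a tank UI entry dot icon, switching between unused and used looks through the IsUsed parameter.
# The icon texture, sampled with 9-slice UVs, supplies a mask that blends from opacity-scaled red channel when unused to green channel when used.
# Color blends from white to orange-red, with an exponential radial gradient adding yellow inside the red-channel area, tuned by Radius and Density.
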Material /Tank/UI/Art/M_UI_Tank_Entries
graph {
 properties {blendMode=BLEND_Additive, materialDomain=MD_UI, pixelDepthOffsetMode=PDOM_Legacy, usedShadingModels=Default Lit}
 nodes [
  n1 = TextureSample(UVs=MaterialFunctionCall() {materialFunction=&GetUserInterfaceUV}.9-Slice UV) {texture=&/T_UI_TankIcon_Dot}
  LinearInterpolate(
   A=Multiply(A=ScalarParameter() {defaultValue=0.25, parameterName=opacity}, B=@n1.R)
   B=@n1.G
   Alpha=ScalarParameter() {defaultValue=1, parameterName=IsUsed}
  )
  n2 = Constant4Vector() {constant=<rgba 1 0.112 0 1>}
  n3 = Constant4Vector() {constant=<rgba 1 0.7522 0 1>}
  n4 = MaterialFunctionCall(
   UVs=MaterialFunctionCall() {materialFunction=&GetUserInterfaceUV}.Normalized UV
   Radius=ScalarParameter() {defaultValue=1, parameterName=Radius}
   Density=ScalarParameter() {defaultValue=1, parameterName=Density}
  ) {materialFunction=&RadialGradientExponential}
  LinearInterpolate(
   A=Constant() {r=1}
   B=LinearInterpolate(A=@n2.RGBA, B=LinearInterpolate(A=@n2.RGBA, B=@n3.RGBA, Alpha=@n4.RadialGradientExponential), Alpha=@n1.R)
   Alpha=ScalarParameter() {defaultValue=1, parameterName=IsUsed}
  )
 ]
}
